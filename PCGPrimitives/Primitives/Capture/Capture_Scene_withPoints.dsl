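; PCG primitive graph that, judging by its name and parameters, captures scene color and normals with points, using planes sized from input bounds or set extents.
; It tags the input, then chains template input management, a capture core subgraph and output preparation.
; Input fallback, input processing, core and output graphs are exposed as swappable parameters.
; Capture direction, resolution, depth, local transform, tag filtering, PCG-content exclusion and plane debugging pass to the core.
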(graph Capture_Scene_withPoints
  (params (InputsProcessingGraph :type (SoftObject /Script/PCG.PCGGraphInterface) :default /PCGPrimitives/Primitives/_Template_/Primitives_InputsProcessing.Primitives_InputsProcessing) (InputsFallbackGraph :type (SoftObject /Script/PCG.PCGGraphInterface) :default /PCGPrimitives/Primitives/_Template_/Primitives_InputsFallback.Primitives_InputsFallback) (CoreProcessGraph :type (SoftObject /Script/PCG.PCGGraphInterface) :default /PCGPrimitives/Primitives/Capture/Subgraphs/Capture_Scene_withPoints_CoreProcess.Capture_Scene_withPoints_CoreProcess) (PrepareOutputsGraph :type (SoftObject /Script/PCG.PCGGraphInterface) :default /PCGPrimitives/Primitives/_Template_/Primitives_PrepareOutputs.Primitives_PrepareOutputs) (UseInputBoundsAsCaptureSource :type Bool :default true) (CaptureFromForwardVector :type Bool) (CaptureExtents :type Vector2 :default (256.0 256.0)) (CaptureTexelSize :type Double :default 1.0) (SampleSize :type Double :default 1e+01) (MaxDepth :type Double :default 256.0) (LocalPositionOffset :type Vector) (LocalRotation :type Rotator) (CaptureColor :type Bool :default true) (CaptureNormal :type Bool :default true) (CaptureOnlyIncludedTags :type Bool) (IncludedTags :type (Array Name)) (ExcludedTags :type (Array Name)) (ExcludePCGGeneratedContent :type Bool) (DebugCapturePlanes :type Bool))
  (inputs (PrimaryInput :type Any))
  (bind AddTags_1 (Add%20Tags :TagsToAdd "Input:0"))
  ;[0 0 0 0.5 32] Inputs
  (bind Subgraph_0 (Subgraph :Graph /PCGPrimitives/Primitives/_Template_/Subgraphs/Primitives_InputsManagement.Primitives_InputsManagement :InputsFallbackGraph InputsFallbackGraph :InputsProcessingGraph InputsProcessingGraph))
  ;[0 0 0 0.5 32] Core Process
  (bind Subgraph_3 (Subgraph :Graph /PCGPrimitives/Primitives/Capture/Subgraphs/Capture_Scene_withPoints_CoreProcess.Capture_Scene_withPoints_CoreProcess :CaptureColor CaptureColor :CaptureExtents CaptureExtents :CaptureFromForwardVector CaptureFromForwardVector :CaptureNormal CaptureNormal :CaptureOnlyIncludedTags CaptureOnlyIncludedTags :CaptureTexelSize CaptureTexelSize :DebugCapturePlanes DebugCapturePlanes :ExcludedTags ExcludedTags :ExcludePCGGeneratedContent ExcludePCGGeneratedContent :IncludedTags IncludedTags :LocalPositionOffset LocalPositionOffset :LocalRotation LocalRotation :MaxDepth MaxDepth :SampleSize SampleSize :SubgraphOverride-param CoreProcessGraph :UseInputBoundsAsCaptureSource UseInputBoundsAsCaptureSource))
  ;[0 0 0 0.5 32] Prepare Output
  (bind Subgraph_4 (Subgraph :Graph /PCGPrimitives/Primitives/_Template_/Primitives_PrepareOutputs.Primitives_PrepareOutputs :SubgraphOverride-param PrepareOutputsGraph))
  (connect DefaultInputNode PrimaryInput AddTags_1 In)
  (connect AddTags_1 Out Subgraph_0 In)
  (connect Subgraph_0 Out Subgraph_3 In)
  (connect Subgraph_3 Out Subgraph_4 In)
  (output Out Subgraph_4))
